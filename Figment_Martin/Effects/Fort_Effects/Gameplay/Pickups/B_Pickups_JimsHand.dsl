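; Fortnite pickup effect Blueprint for Jim's Hand, a gauntlet pickup that spins and shows rarity-colored visuals.
; Construction builds a skeletal or static mesh and derives color from item rarity or an override curve.
; Pickup drives timestamped hologram materials and pooled stretch MIDs, with separate backpack, Bluglo, Violentium and gift handling.
; A curve-table despawn time triggers fading particles; destruction before pickup plays derez FX and sound.
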(event EventBeginPlay
  (bind (_asfort_game_state_athena _bsuccess) (Utilities|Casting|CastToFortGameStateAthena (Game|GetGameState)))
  (bind _itemdefinition (Variables|PickupProperties|GetItemDefinition))
  (bind _skeletal_mesh_pickup (Variables|Default|GetSkeletal_Mesh_Pickup))
  (bind _skeletalorstaticmeshassetprimitive (Variables|Default|GetSkeletalOrStaticMeshAssetPrimitive))
  ;[1 1 1] Begin Play
  (Utilities|FlowControl|Sequence
    (:then_0
      ;; Jon's Parenting Stuff
      (Utilities|Casting|CastToFortPickupBackpack (Actor|GetOwner)
        (:then
          (Variables|Default|SetIs%20a%20Backpack true))
        (:CastFailed
          _))
      (if (Variables|Default|GetIsSkeletalMesh)
        (Transformation|SetRelativeLocation _skeletal_mesh_pickup (Variables|Default|GetPivot_Offset) false false)
        (Transformation|SetRelativeRotation _skeletal_mesh_pickup (Math|Rotator|SelectRotator "0, 0, 0" (Variables|Default|GetRotation_Bias) (Variables|Default|GetIs%20a%20Backpack)) false false))
      (Collision|SetCollisionEnabled _skeletalorstaticmeshassetprimitive "NoCollision")
      (Components|Tick|SetComponentTickEnabled _skeletalorstaticmeshassetprimitive false)
      (Transformation|AttachComponentToComponent (Variables|Default|GetSkeletalOrStaticMeshAssetPrimitive) (Variables|Default|GetRoot) "None" "KeepRelative" "KeepRelative" "KeepRelative" true)
      (Utilities|IsValid _itemdefinition
        (:Is%20Valid
          (bind _meshtransform (Class|FortItemFunctionLibrary|GetPickupMeshTransform _itemdefinition)
            (:IsIdentity
              (Transformation|SetWorldLocation (Variables|Default|GetSkeletalOrStaticMeshAssetPrimitive) (+ (Transformation|GetActorLocation) (- (Transformation|GetWorldLocation (Variables|Default|GetSkeletalOrStaticMeshAssetPrimitive)) (Collision|GetComponentBounds (Variables|Default|GetSkeletalOrStaticMeshAssetPrimitive)))) false false :!label _label))
            (:IsOverridden
              (Transformation|SetRelativeTransform (Variables|Default|GetSkeletalOrStaticMeshAssetPrimitive) _meshtransform false false)))
          (Utilities|Array|Resize (Variables|Default|GetmidArray) (Rendering|Material|GetNumMaterials (Variables|Default|GetSkeletalOrStaticMeshAssetPrimitive))))
        (:Is%20Not%20Valid
          _))
      (goto _label))
    (:then_1
      (CallFunction|SetupBaseMIDs))
    (:then_2
      ;;[1 1 1 1 12] Pre-create one entry in the stretch pool
      (CallFunction|SetupStretchMIDs)
      (Default|ReleaseStretchMIDs))
    (:then_3
      ;; Set draw distance
      (if (not (Variables|Default|GetIsBluglo))
        (bind _output_get (Variables|Default|SetViewDistance (select _bsuccess (select (Fortnite|IsMobilePlatform) (Variables|MaxDrawDistance|GetMax%20Draw%20Distance%20Bacchus) (Variables|MaxDrawDistance|GetMax%20Draw%20Distance%20Athena)) (Variables|MaxDrawDistance|GetMax%20Draw%20Distance%20StW))))
        (LOD|SetMaxDrawDistance (Variables|Default|GetSkeletalOrStaticMeshAssetPrimitive) (* _output_get (Utilities|Reciprocal (Fortnite|GetNumericCVar "r.ViewDistanceScale"))))
        (bind _backgroundparticlesystem (Variables|Default|GetBackgroundParticleSystem)
          (:then
            (LOD|SetMaxDrawDistance _backgroundparticlesystem 7000))
          (:else
            _)))
      ;; Randomize Mesh Rotation for BR
      (if (Variables|PickupProperties|GetbRandomRotation)
        (switch Utilities|FlowControl|Switch|SwitchonESubGame (Fortnite|GetSubGame)
          (:Campaign
            _)
          (:Athena
            (bind _output_get_2 (Variables|Default|SetRandom%20Rotation (Math|Random|RandomFloatinRange 0.0 3.6e+02)))
            (Transformation|SetActorRotation 0 (Math|Rotator|MakeRotator 0.0 0.0 _output_get_2) false)
            (Transformation|SetRelativeRotation (Variables|Default|GetBackground) (Math|Rotator|MakeRotator 0.0 0.0 (* _output_get_2 -1.0)) false false))
          (:Invalid
            _)))
      ;; MobileInteractIconSetup
      (Variables|Default|SetMobileInteractionMaterial (Utilities|Select :Index (Variables|Default|GetRarity%20Level) :Option%200 "/Game/Effects/Fort_Effects/Gameplay/Pickups/Materials/M_MobileInteractIcon.M_MobileInteractIcon" :Option%201 "/Game/Effects/Fort_Effects/Gameplay/Pickups/Materials/M_MobileInteractIcon_Tier01_Inst.M_MobileInteractIcon_Tier01_Inst" :Option%202 "/Game/Effects/Fort_Effects/Gameplay/Pickups/Materials/M_MobileInteractIcon_Tier02_Inst.M_MobileInteractIcon_Tier02_Inst" :Option%203 "/Game/Effects/Fort_Effects/Gameplay/Pickups/Materials/M_MobileInteractIcon_Tier03_Inst.M_MobileInteractIcon_Tier03_Inst" :Option%204 "/Game/Effects/Fort_Effects/Gameplay/Pickups/Materials/M_MobileInteractIcon_Tier04_Inst.M_MobileInteractIcon_Tier04_Inst")))
    (:then_4
      (Networking|CanExecuteCosmeticEvents
        (:True
          (bind _outxy (DataTable|EvaluateCurveTableRow "/Game/Athena/Balance/DataTables/AthenaGameData.AthenaGameData" "Default.MeteorDrop.Gauntlet.DespawnTime")
            (:RowFound
              (Utilities|FlowControl|Delay (- _outxy 5.0))
              (if (not (Variables|Default|GetPickedUp))
                (Effects|Components|ParticleSystem|SetTemplate (Variables|Default|GetP_Jim_Gauntlet_Pickup) (Variables|Default|GetFading_FX))
                (CallFunction|SetRarityColor)))))))))

(event EventOnPickedUp
  (params (PickupTarget :type Fort%20Pawn%20Object%20Reference))
  (bind _itemdefinition (Variables|PickupProperties|GetItemDefinition))
  ;[1 1 1] item picked up by user
  (Utilities|FlowControl|Sequence
    (:then_0
      (Variables|Default|SetPickedUp true)
      (switch Utilities|FlowControl|Switch|SwitchonESubGame (Fortnite|GetSubGame)
        (:Campaign
          (if (not (Variables|PickupProperties|GetbOwnedByPlayer))
            (Utilities|Casting|CastToFortResourceItemDefinition _itemdefinition
              (:CastFailed
                _)))
          (if (not (>= (Variables|Default|GetRarity%20Level) 1))
            (Utilities|Casting|CastToFortIngredientItemDefinition _itemdefinition))
          (Utilities|FlowControl|Delay 0.25)
          (Effects|SpawnEmitteratLocation "/Game/Effects/Fort_Effects/Effects/LootContainers/P_Loot_Effect_02.P_Loot_Effect_02" (Transformation|GetWorldLocation (Variables|Default|GetSkeletalOrStaticMeshAssetPrimitive))))))
    (:then_1
      (CallFunction|SetActiveBackgroundParticleSystem false)
      (CallFunction|SetHiddenBackgroundVisualComponents true)
      (CallFunction|SetVisibleMobileInteractIcon false))
    (:then_2
      ;; Material / FX
      (if (Variables|Default|GetIsLootGiftForOthers)
        _
        (elif (== (Variables|Default|GetRarity%20Level) 0)
          (CallFunction|SetHologramPickedUpParams true)
          (Utilities|FlowControl|Delay (Variables|HologramSetup|GetPicked%20Up%20Fade%20Out%20Time))
          (else
            (switch Utilities|FlowControl|Switch|SwitchonESubGame (Fortnite|GetSubGame)
              (:Campaign
                (CallFunction|SpawnPickedUpTrailPS :!label _label))
              (:Athena
                _)
              (:Invalid
                (goto _label)))
            (Utilities|FlowControl|Delay (Variables|HologramSetup|GetPicked%20Up%20Delay%20Before%20Glow))
            (CallFunction|SetHologramPickedUpParams false)
            (Utilities|FlowControl|Delay (Math|Float|float+float (Variables|HologramSetup|GetPicked%20Up%20Fade%20In%20Time) (Variables|HologramSetup|GetPicked%20Up%20Hold%20Time) (Variables|HologramSetup|GetPicked%20Up%20Fade%20Out%20Time) 0.05))))))
    (:then_3
      (bind _backgroundparticlesystem (Variables|Default|GetBackgroundParticleSystem)
        (:then
          (Variables|/Script/CascadeRuntime.ParticleSystemComponent|SetCustomTimeDilation (Utilities|Select :Index (Fortnite|GetSubGame) :Campaign 3.0 :Athena 9.0 :Invalid 3.0 :Count 0.0 :Creative 0.0) _backgroundparticlesystem)
          (CallFunction|SetupStretchMIDs))
        (:else)))))

(event EventOnAboutToEnterBackpack
  (params (PickupTarget :type Fort%20Pawn%20Object%20Reference))
  ;[1 1 1] End animation and hide asset
  (Utilities|IsValid PickupTarget
    (:Is%20Valid
      (Rendering|SetActorHiddenInGame true)
      ;[1 1 1] For pickups that were on the ground, apply small force feedback as pickup enters backpack
      (if (and (Variables|Default|GetTossed) (Pawn|IsLocallyControlled PickupTarget))
        (bind _asfort_player_controller (Utilities|Casting|CastToFortPlayerController (Pawn|GetController PickupTarget)))
        (Game|Feedback|ClientPlayForceFeedback _asfort_player_controller (Variables|Default|GetPickupForceFeedback) "Pickup" false false false)))))

(event EventTick
  (params (DeltaSeconds :type Float%20%28single-precision%29))
  ;[1 1 1] Event tick
  (Networking|CanExecuteCosmeticEvents
    (:True
      (Transformation|AddLocalRotation (Variables|Default|GetSkeletalOrStaticMeshAssetPrimitive) (Math|Rotator|MakeRotator 0.0 0.0 (* DeltaSeconds 28.0)) false false))))

(event EventOnTossed
  (bind _backpackparticlesystem (Variables|VFX|GetBackpackParticleSystem))
  (bind _blugloparticlesystem (Variables|VFX|GetBlugloParticleSystem))
  (bind _violentium_idle_fx (Variables|VFX|GetViolentium_Idle_FX))
  ;[0 0 0] Activate particle system if tossed and it should exist
  (Variables|Default|SetTossed true)
  (if (Variables|VFX|GetActivate%20Rarity%20Particle%20Systems)
    (if (Variables|Default|GetIs%20a%20Backpack)
      ;; Spawn Backpack Particles
      (if (not (Asset|IsAssetNull _backpackparticlesystem))
        (bind _object (Utilities|AsyncLoadAsset _backpackparticlesystem)
          (:Completed
            (bind _asparticle_system (Utilities|Casting|CastToParticleSystem _object))
            (bind _returnvalue (Effects|SpawnEmitterAttached _asparticle_system))
            (bind _output_get (Variables|Default|SetBackgroundParticleSystem _returnvalue))
            (LOD|SetMaxDrawDistance _output_get 5e+03)
            (CallFunction|DestroyBackgroundVisualComponents))))
      (elif (Variables|Default|GetIsBluglo)
        ;; Spawn Bluglo Particles
        (if (not (Asset|IsAssetNull _blugloparticlesystem))
          (bind _object_2 (Utilities|AsyncLoadAsset _blugloparticlesystem)
            (:Completed
              (bind _asparticle_system_2 (Utilities|Casting|CastToParticleSystem _object_2))
              (bind _returnvalue_2 (Effects|SpawnEmitterAttached _asparticle_system_2 (Variables|Default|GetRoot)))
              (bind _output_get_2 (Variables|Default|SetBackgroundParticleSystem _returnvalue_2))
              (LOD|SetMaxDrawDistance _output_get_2 5e+03)
              (CallFunction|DestroyBackgroundVisualComponents))))
        (elif (Variables|Default|GetIs%20Violentium)
          ;; Spawn Violentium Particles
          (if (not (Asset|IsAssetNull _violentium_idle_fx))
            (bind _object_3 (Utilities|AsyncLoadAsset _violentium_idle_fx)
              (:Completed
                (Utilities|Casting|CastToParticleSystem _object_3)
                (bind _returnvalue_3 (Effects|SpawnEmitterAttached 0))
                (bind _output_get_3 (Variables|Default|SetBackgroundParticleSystem _returnvalue_3))
                (LOD|SetMaxDrawDistance _output_get_3 5e+03)
                (CallFunction|DestroyBackgroundVisualComponents))))
          (else
            (bind (_bneedstobeinitialized _returnvalue_4) (FortFX|GetPickupRarityMID (Variables|Default|GetBackground) 0 (Variables|Default|GetRarity%20Level)))
            (if _bneedstobeinitialized
              (CallFunction|SetInitialBGRaysParams _returnvalue_4))))))
    (else
      (CallFunction|SetActiveBackgroundParticleSystem false))))

(event EventDestroyed
  ;[1 1 1] Return mesh MID when destroyed
  (|CallParentFunction)
  (Default|ReleaseStretchMIDs))

(event EventEnableBacchusHighlight
  (bind _root (Variables|Default|GetRoot))
  ;[1 1 1] Mobile highlighting
  (Networking|CanExecuteCosmeticEvents
    (:True
      (bind (_direction _lerpobject) (Timeline|MobileSelectedTL Play)
        (:Update
          (Transformation|SetRelativeScale3D (Variables|Default|GetRoot) (Math|Conversions|ToVector(Float) _lerpobject)))
        (:Finished
          (switch Utilities|FlowControl|Switch|SwitchonETimelineDirection _direction
            (:Forward
              (Transformation|SetRelativeScale3D _root (Variables|Default|GetMobileSelectedScale)))
            (:Backward
              (Transformation|SetRelativeLocation (Variables|Default|GetRoot) "0, 0, 0" false false)
              (Transformation|SetRelativeScale3D _root "1.000000,1.000000,1.000000"))))))))

(event EventDisableBacchusHighlight
  (bind _root (Variables|Default|GetRoot))
  ;[1 1 1] Mobile highlighting
  (Networking|CanExecuteCosmeticEvents
    (:True
      (if (not (Variables|Default|GetPickedUp))
        (Timeline|MobileSelectedTL Reverse)))))

(event EventEndPlay
  (params (EndPlayReason :type EEndPlayReason%20Enum :default "Destroyed"))
  (bind _skeletalorstaticmeshassetprimitive (Variables|Default|GetSkeletalOrStaticMeshAssetPrimitive))
  (bind _returnvalue (Transformation|GetWorldLocation _skeletalorstaticmeshassetprimitive))
  (Networking|CanExecuteCosmeticEvents
    (:True
      (switch Utilities|FlowControl|Switch|SwitchonEEndPlayReason EndPlayReason
        (:Destroyed
          (if (not (Variables|Default|GetPickedUp))
            (Effects|SpawnEmitteratLocation "/Figment_Martin/Athena/Items/Gameplay/BackPacks/CarminePack/FX/P_Jim_Gauntlet_Pickup_Derez.P_Jim_Gauntlet_Pickup_Derez" _returnvalue (Transformation|GetWorldRotation _skeletalorstaticmeshassetprimitive))
            (Audio|PlaySoundatLocation (Variables|Default|GetDisappearSound) _returnvalue)))))))

(fn DestroyBackgroundVisualComponents
  (bind _background (Variables|Default|GetBackground)
    (:then
      (Components|DestroyComponent _background))
    (:else
      _))
  (bind _spotlight (Variables|Default|GetSpotLight)
    (:then
      (Components|DestroyComponent _spotlight))
    (:else)))

(fn ReleaseStretchMIDs
  (bind _midsneedingrelease (Variables|Default|GetMIDsNeedingRelease))
  (bind _skeletalorstaticmeshassetprimitive (Variables|Default|GetSkeletalOrStaticMeshAssetPrimitive))
  (for _array_element _midsneedingrelease
    (FortFX|ReturnPickupStretchMIDToPool _skeletalorstaticmeshassetprimitive _array_element)
    (Rendering|Material|SetMaterial _skeletalorstaticmeshassetprimitive _array_element 0))
  (Utilities|Array|Clear _midsneedingrelease))

(fn ScaleHologramTimingsForPvP
  (switch Utilities|FlowControl|Switch|SwitchonESubGame (Fortnite|GetSubGame)
    (:Athena
      (Variables|HologramSetup|SetPicked%20Up%20Delay%20Before%20Glow (* (Variables|HologramSetup|GetPicked%20Up%20Delay%20Before%20Glow) (LocalVariables|GetPVP_TimeScale)))
      (Variables|HologramSetup|SetPicked%20Up%20Fade%20In%20Time (* (Variables|HologramSetup|GetPicked%20Up%20Fade%20In%20Time) (LocalVariables|GetPVP_TimeScale)))
      (Variables|HologramSetup|SetPicked%20Up%20Hold%20Time (* (Variables|HologramSetup|GetPicked%20Up%20Hold%20Time) (LocalVariables|GetPVP_TimeScale)))
      (Variables|HologramSetup|SetPicked%20Up%20Fade%20Out%20Time (* (Variables|HologramSetup|GetPicked%20Up%20Fade%20Out%20Time) (LocalVariables|GetPVP_TimeScale)))
      (Variables|HologramSetup|SetPicked%20Up%20Sweep%20Speed (* (Variables|HologramSetup|GetPicked%20Up%20Sweep%20Speed) (LocalVariables|GetPVP_TimeScale))))))

(fn SetActiveBackgroundParticleSystem
  (params (Active :type Boolean) (Reset :type Boolean))
  (bind _backgroundparticlesystem (Variables|Default|GetBackgroundParticleSystem)
    (:then
      (Components|Activation|SetActive _backgroundparticlesystem Active Reset))
    (:else)))

(fn SetHiddenBackgroundVisualComponents
  (params (Hidden :type Boolean))
  (bind _background (Variables|Default|GetBackground)
    (:then
      (Development|SetHiddeninGame _background Hidden true))
    (:else
      _))
  (bind _spotlight (Variables|Default|GetSpotLight)
    (:then
      (Development|SetHiddeninGame _spotlight Hidden true))
    (:else)))

(fn SetHologramPickedUpParams
  (params (Tier%200 :type Boolean))
  (bind _returnvalue (Utilities|Time|GetGameTimeinSeconds))
  (bind _midarray (Variables|Default|GetmidArray))
  (if Tier%200
    ;[1 1 1] For common
    (for _array_element _midarray
      (Utilities|IsValid _array_element
        (:Is%20Valid
          (Rendering|Material|SetScalarParameterValue _array_element "TimestampB" _returnvalue :!class /Script/Engine.MaterialInstanceDynamic)
          (Rendering|Material|SetScalarParameterValue _array_element "TransitionLength_Out" (Variables|HologramSetup|GetPicked%20Up%20Fade%20Out%20Time) :!class /Script/Engine.MaterialInstanceDynamic)
          (Rendering|Material|SetScalarParameterValue _array_element "Sweep Speed" 0.0 :!class /Script/Engine.MaterialInstanceDynamic)
          (Rendering|Material|SetScalarParameterValue _array_element "Random" (* (Math|Float|Fraction _returnvalue) -1) :!class /Script/Engine.MaterialInstanceDynamic))))
    (else
      ;[1 1 1] All others
      (for _array_element_2 _midarray
        (Utilities|IsValid _array_element_2
          (:Is%20Valid
            (Rendering|Material|SetScalarParameterValue _array_element_2 "TimestampA" (Utilities|Time|GetGameTimeinSeconds) :!class /Script/Engine.MaterialInstanceDynamic)
            (Rendering|Material|SetScalarParameterValue _array_element_2 "TransitionLength_In" (Variables|HologramSetup|GetPicked%20Up%20Fade%20In%20Time) :!class /Script/Engine.MaterialInstanceDynamic)
            (Rendering|Material|SetScalarParameterValue _array_element_2 "TimestampB" (Math|Float|float+float (Variables|HologramSetup|GetPicked%20Up%20Hold%20Time) (Variables|HologramSetup|GetPicked%20Up%20Fade%20In%20Time) (Utilities|Time|GetGameTimeinSeconds)) :!class /Script/Engine.MaterialInstanceDynamic)
            (Rendering|Material|SetScalarParameterValue _array_element_2 "TransitionLength_Out" (Variables|HologramSetup|GetPicked%20Up%20Fade%20Out%20Time) :!class /Script/Engine.MaterialInstanceDynamic)
            (Rendering|Material|SetScalarParameterValue _array_element_2 "Sweep Speed" (Variables|HologramSetup|GetPicked%20Up%20Sweep%20Speed) :!class /Script/Engine.MaterialInstanceDynamic)
            (Rendering|Material|SetScalarParameterValue _array_element_2 "Random" (* (Math|Float|Fraction (Utilities|Time|GetGameTimeinSeconds)) -1) :!class /Script/Engine.MaterialInstanceDynamic))))))
  (return))

(fn SetInitialBGRaysParams
  (params (MID :type Material%20Instance%20Dynamic%20Object%20Reference))
  (bind _viewdistance (Variables|Default|GetViewDistance))
  (Rendering|Material|SetVectorParameterValue MID "Color" (Utilities|Array|Get(aref) (Variables|HologramSetup|GetBackgroundRarityColors) (Variables|Default|GetRarity%20Level)) :!class /Script/Engine.MaterialInstanceDynamic)
  (Rendering|Material|SetScalarParameterValue MID "End Fade" _viewdistance :!class /Script/Engine.MaterialInstanceDynamic)
  (Rendering|Material|SetScalarParameterValue MID "Start Fade" (* _viewdistance 0.85) :!class /Script/Engine.MaterialInstanceDynamic)
  (Rendering|Material|SetScalarParameterValue MID "Random" (Math|Random|RandomFloatinRange 0.0 1e+01) :!class /Script/Engine.MaterialInstanceDynamic)
  (if (>= (Variables|Default|GetRarity%20Level) 4)
    (Rendering|Material|SetScalarParameterValue MID "IsLegendaryRarity" 1.0 :!class /Script/Engine.MaterialInstanceDynamic)))

(fn SetInitialHologramParams
  (params (MID :type Material%20Instance%20Dynamic%20Object%20Reference))
  (if (not (Variables|Unique_Material_Check|GetHasUniqueMaterialIds))
    (Rendering|Material|SetVectorParameterValue MID "Color" (Variables|HologramSetup|GetRarity%20Color) :!class /Script/Engine.MaterialInstanceDynamic)
    (Rendering|Material|SetScalarParameterValue MID "Random" (Math|Random|RandomFloatinRange 0.0 10) :!class /Script/Engine.MaterialInstanceDynamic)
    (Rendering|Material|SetScalarParameterValue MID "Sweep Brightness" (Utilities|Array|Get(acopy) (Variables|HologramSetup|GetSweep%20Brightness) (Variables|Default|GetRarity%20Level)) :!class /Script/Engine.MaterialInstanceDynamic)
    (Rendering|Material|SetScalarParameterValue MID "Sweep Speed" (Utilities|Array|Get(acopy) (Variables|HologramSetup|GetSweep%20Speed) (Variables|Default|GetRarity%20Level)) :!class /Script/Engine.MaterialInstanceDynamic)
    (Rendering|Material|SetScalarParameterValue MID "Hologram_Control" 1.0 :!class /Script/Engine.MaterialInstanceDynamic)
    (return)))

(fn SetRarityColor
  (bind _secondarycolorcurvelookup (Variables|Default|GetSecondaryColorCurveLookup))
  (bind _coloroverridecurve (Variables|Default|GetColorOverrideCurve))
  (bind _returnvalue (Math|Curves|GetLinearColorValue _coloroverridecurve _secondarycolorcurvelookup))
  (bind _returnvalue_2 (Math|Curves|GetLinearColorValue _coloroverridecurve 0.0))
  (bind _itemdefinition (Variables|PickupProperties|GetItemDefinition))
  (bind _returnvalue_3 (Class|FortItemFunctionLibrary|CreateTemporaryItemInstance _itemdefinition))
  (bind _returnvalue_4 (Item|GetRarity _returnvalue_3))
  (bind _p_jim_gauntlet_pickup (Variables|Default|GetP_Jim_Gauntlet_Pickup))
  (bind _rarity_level (Variables|Default|GetRarity%20Level))
  (Utilities|IsValid (Variables|Default|GetColorOverrideCurve)
    (:Is%20Valid
      (Variables|Default|SetRarity%20Level 3)
      (Utilities|FlowControl|Sequence
        (:then_0
          (Variables|HologramSetup|SetRarity%20Color _returnvalue_2)
          (Utilities|Array|SetArrayElem (Variables|HologramSetup|GetOutline%20Rarity%20Colors) _rarity_level (Math|Color|LinearColor*Float _returnvalue_2 2.2))
          (Effects|Components|ParticleSystem|SetColorParameter _p_jim_gauntlet_pickup "Color1" _returnvalue_2))
        (:then_1
          (Utilities|Array|SetArrayElem (Variables|HologramSetup|GetBackgroundRarityColors) _rarity_level _returnvalue)
          (Effects|Components|ParticleSystem|SetColorParameter _p_jim_gauntlet_pickup "Color2" _returnvalue))))
    (:Is%20Not%20Valid
      (Variables|PickupProperties|GetItemDefinition
        (:then
          (Fortnite|GetRarityData _returnvalue_4)
          (bind _output_get (Variables|Default|SetRarity%20Level (Utilities|Select :Index _returnvalue_4 :Common 0 :Uncommon 1 :Rare 2 :Epic 3 :Legendary 4 :Mythic 5 :Transcendent 6 :Unattainable 7)))
          (Variables|HologramSetup|SetRarity%20Color (Math|Color|SelectColor (Utilities|Array|Get(acopy) (Variables|HologramSetup|GetOutline%20Rarity%20Colors) _output_get) (Utilities|Array|Get(acopy) (Variables|HologramSetup|GetOutline%20Rarity%20Colors)) (Utilities|Array|IsValidIndex (Variables|HologramSetup|GetOutline%20Rarity%20Colors) _output_get)))
          (return))
        (:else
          (Variables|Default|SetRarity%20Level 0))))))

(fn SetupBaseMIDs
  (bind _currentmaterial (Variables|Default|GetCurrentMaterial))
  (for (_ _array_index) (Variables|Default|GetmidArray)
    (bind _output_get (Variables|Default|SetCurrentElementIndex _array_index))
    (if (not (Math|Boolean|ORBoolean (Variables|Default|GetIsBluglo) (Variables|Default|GetIs%20a%20Backpack) (Variables|Default|GetIsLootGiftForOthers)))
      (Utilities|Array|SetArrayElem _currentmaterial _output_get (Rendering|Material|GetMaterial (Variables|Default|GetSkeletalOrStaticMeshAssetPrimitive) (Variables|Default|GetCurrentElementIndex)) true)
      (bind _returnvalue (Rendering|Material|GetBaseMaterial (Utilities|Array|Get(acopy) _currentmaterial (Variables|Default|GetCurrentElementIndex))))
      (Utilities|Array|SetArrayElem (Variables|Default|GetCurrentParentMaterial) (Variables|Default|GetCurrentElementIndex) _returnvalue true)))
  (CallFunction|UniqueMaterialMIDs)
  (return))

(fn SetupLight
  (Utilities|EffectsQualityCheck 0
    (:Cinematic
      _)
    (:Epic
      _)
    (:High
      (Components|DestroyComponent (Variables|Default|GetSpotLight) :!label _label))
    (:Medium
      (goto _label))
    (:Low
      (goto _label)))
  (if (or (Fortnite|IsMobilePlatform) (== (Variables|Default|GetRarity%20Level) 0))
    (goto _label)
    (else
      (Rendering|Components|Light|SetLightColor (Variables|Default|GetSpotLight) (Variables|HologramSetup|GetRarity%20Color) true)
      (Rendering|SetCastShadow (Variables|Default|GetSkeletalOrStaticMeshAssetPrimitive) false))))

(fn SetupStretchMIDs
  (bind _currentelementindex (Variables|Default|GetCurrentElementIndex))
  (bind _midarray (Variables|Default|GetmidArray))
  (bind _output (Utilities|Array|Get(acopy) _midarray _currentelementindex))
  (bind _currentelementindex_2 (Variables|Default|GetCurrentElementIndex))
  (bind _midarray_2 (Variables|Default|GetmidArray))
  (bind _output_2 (Utilities|Array|Get(acopy) _midarray_2 _currentelementindex_2))
  (bind _currentelementindex_3 (Variables|Default|GetCurrentElementIndex))
  (bind _midarray_3 (Variables|Default|GetmidArray))
  (bind _output_3 (Utilities|Array|Get(acopy) _midarray_3 _currentelementindex_3))
  (bind _currentelementindex_4 (Variables|Default|GetCurrentElementIndex))
  (bind _skeletalorstaticmeshassetprimitive (Variables|Default|GetSkeletalOrStaticMeshAssetPrimitive))
  (for (_ _array_index) (Variables|Default|GetmidArray)
    (bind _output_get (Variables|Default|SetCurrentElementIndex _array_index))
    (if (Utilities|Array|IsValidIndex (Variables|Default|GetCurrentMaterial) _output_get)
      (if (Variables|Default|GetIs%20a%20Backpack)
        ;[0.335 0.5961 1] If Backpack Set up Glow
        (bind _returnvalue (Rendering|Material|CreateDynamicMaterialInstance "/Game/Packages/Fortnite_UniqueMaterials_1/BackPackShrinkingEffect_Super_Charge.BackPackShrinkingEffect_Super_Charge" "None" "None" :!class /Script/Engine.KismetMaterialLibrary))
        (Utilities|Array|SetArrayElem (Variables|Default|GetmidArray) (Variables|Default|GetCurrentElementIndex) _returnvalue true)
        (Rendering|Material|CopyMaterialInstanceParameters _output (Utilities|Array|Get(acopy) (Variables|Default|GetCurrentMaterial) _currentelementindex) true)
        (Rendering|Material|SetScalarParameterValue _output "SuperChargedEffect" 1 :!class /Script/Engine.MaterialInstanceDynamic)
        (Rendering|Material|SetScalarParameterValue _output "SuperChargedEffect" 1 :!class /Script/Engine.MaterialInstanceDynamic)
        (else
          ;[0.508 0.6919 1] Reparent Most Materials
          (bind (_bneedstobeinitialized _returnvalue_2) (FortFX|GetPickupStretchMIDFromPool (select (Variables|Unique_Material_Check|GetHasUniqueMaterialIds) (Utilities|Array|Get(acopy) (Variables|Default|GetCurrentMaterial) (Variables|Default|GetCurrentElementIndex)) "/Game/Packages/Fortnite_UniqueMaterials_1/BackPackShrinkingEffect.BackPackShrinkingEffect") (Variables|Default|GetSkeletalOrStaticMeshAssetPrimitive) _currentelementindex_4))
          (Utilities|Array|SetArrayElem (Variables|Default|GetmidArray) _currentelementindex_4 _returnvalue_2 true)
          (if _bneedstobeinitialized
            (Rendering|Material|CopyMaterialInstanceParameters _output_3 (Utilities|Array|Get(acopy) (Variables|Default|GetCurrentMaterial) _currentelementindex_3) true)
            (CallFunction|SetInitialHologramParams _output_3)))))
    (Utilities|IsValid _output_2
      (:Is%20Valid
        (Rendering|Material|SetMaterial _skeletalorstaticmeshassetprimitive _currentelementindex_2 _output_2)
        (Rendering|Material|SetScalarParameterValue _output_2 "Use Distance Based Distance Cull Scale Override For Pickup Effects" 1 :!class /Script/Engine.MaterialInstanceDynamic))
      (:Is%20Not%20Valid
        _))
    ;[1 1 1] gift for other players. can't interact
    (if (Variables|PickupProperties|GetbOwnedByPlayer)
      (if (not (Variables|PickupProperties|GetbOwnedByALocalPlayer))
        ;[1 0.07181 0.2116 1 12] We need to release any stretch MIDs before overriding the material. Ideally we wouldn't even make the strrech MID for element 0 for gifts. Easy to move?
        (Default|ReleaseStretchMIDs)
        (if (Variables|Default|GetIs%20a%20Backpack)
          (Rendering|Material|SetMaterial (Variables|Default|GetSkeletalOrStaticMeshAssetPrimitive) 0 "/Game/Missions/Rewards/Materials/M_LootGift_Skel.M_LootGift_Skel")
          (Variables|VFX|SetActivate%20Rarity%20Particle%20Systems false)
          (else
            (Rendering|Material|SetMaterial _skeletalorstaticmeshassetprimitive 0 "/Game/Missions/Rewards/Materials/M_LootGift_Others.M_LootGift_Others")
            (Variables|VFX|SetActivate%20Rarity%20Particle%20Systems false)
            (Variables|Default|SetIsLootGiftForOthers true)))))))

(fn SetVisibleMobileInteractIcon
  (params (Visible :type Boolean))
  (if Visible
    ;[1 1 1] Create
    (bind _mobileinteracticon (Variables|Default|GetMobileInteractIcon)
      (:then
        (Rendering|SetVisibility _mobileinteracticon true false))
      (:else
        (Variables|Default|GetMobileInteractionMaterial
          (:then
            (bind _returnvalue (AddComponent|NODE_AddStaticMeshComponent-3 :RelativeTransform_Location (Variables|Default|GetMobileInteractIconLocation) :RelativeTransform_Rotation_Yaw (* (Variables|Default|GetRandom%20Rotation) -1.0) :RelativeTransform_Scale (Variables|Default|GetMobileInteractIconScale)))
            (bind _output_get (Variables|Default|SetMobileInteractIcon _returnvalue))
            (Rendering|Material|SetMaterial _output_get 0 (Variables|Default|GetMobileInteractionMaterial)))
          (:else))))
    (else
      ;[1 1 1] Destroy
      (bind _mobileinteracticon_2 (Variables|Default|GetMobileInteractIcon)
        (:then
          (Rendering|SetVisibility _mobileinteracticon_2 false false)
          (Components|DestroyComponent _mobileinteracticon_2))
        (:else)))))

(fn SpawnPickedUpTrailPS
  (bind _returnvalue (Effects|SpawnEmitterAttached (Variables|VFX|GetPicked%20Up%20Trail%20PS) (Variables|Default|GetRoot)))
  (LocalVariables|SetTrailPS _returnvalue)
  (Effects|Components|ParticleSystem|SetFloatParameter (LocalVariables|GetTrailPS) "SPU Scale" (Utilities|Array|Get(acopy) (Variables|VFX|GetSparkle%20Spawn%20Rate%20%28Picked%20Up%29) (Variables|Default|GetRarity%20Level)))
  (Effects|Components|ParticleSystem|SetFloatParameter (LocalVariables|GetTrailPS) "SparkleSphereRadius" (Variables|VFX|GetComponent%20Radius%20%28Scaled%29))
  (Effects|Components|ParticleSystem|SetFloatParameter (LocalVariables|GetTrailPS) "Lifetime" (Utilities|Array|Get(acopy) (Variables|VFX|GetLifetime%20%28Picked%20Up%29) (Variables|Default|GetRarity%20Level)))
  (Effects|Components|ParticleSystem|SetVectorParameter (LocalVariables|GetTrailPS) "Color" (Math|Conversions|ToVector(LinearColor) (Variables|HologramSetup|GetRarity%20Color)))
  (return))

(fn UniqueMaterialMIDs
  (Utilities|Array|AppendArray (LocalVariables|GetAppendedMaterialsArray) (Variables|Default|GetCurrentParentMaterial))
  (Utilities|Array|AppendArray (LocalVariables|GetAppendedMaterialsArray) (Variables|Default|GetCurrentMaterial))
  ;[1 1 1] Unique materials MIDs
  (bind _array_element (Utilities|Array|ForEachLoopwithBreak (LocalVariables|GetAppendedMaterialsArray) :!label _label)
    (:LoopBody
      (if (Utilities|Array|ContainsItem (Variables|Unique_Material_Check|GetNames%20of%20Materials%20And%20Parents%20Exempt%20from%20Repartening) (Utilities|GetObjectName _array_element))
        (Variables|Unique_Material_Check|SetHasUniqueMaterialIds true)
        (goto _label Break)))
    (:Completed
      (return))))

(fn ConstructionScript
  (CallFunction|SetRarityColor)
  (CallFunction|ScaleHologramTimingsforPvP)
  (Utilities|IsValid (Variables|PickupProperties|GetSkeletalMesh)
    (:Is%20Valid
      (Variables|Default|SetIsSkeletalMesh true)
      (bind _returnvalue (AddComponent|NODE_AddSkeletalMeshComponent-0 :bManualAttachment true))
      (bind _output_get (Variables|Default|SetSkeletal_Mesh_Pickup _returnvalue))
      (Components|SkinnedMesh|SetRenderStatic _output_get (Variables|PickupProperties|GetbDoNotTickSkeletalMeshComponents))
      (Components|SkinnedMesh|SetSkinnedAssetandUpdate _output_get (Variables|PickupProperties|GetSkeletalMesh) true))
    (:Is%20Not%20Valid
      (Variables|Default|SetIsSkeletalMesh false)
      (bind _returnvalue_2 (AddComponent|NODE_AddStaticMeshComponent-0 :bManualAttachment true))
      (bind _output_get_2 (Variables|Default|SetStatic_Mesh_Pickup _returnvalue_2))
      (Components|StaticMesh|SetStaticMesh _output_get_2 (Variables|PickupProperties|GetStaticMesh))
      (Transformation|AttachComponentToComponent (Variables|Default|GetP_Jim_Gauntlet_Pickup) _output_get_2 "None" "SnapToTarget" "SnapToTarget" "KeepWorld" true)))
  (bind _output_get_3 (Variables|Default|SetSkeletalOrStaticMeshAssetPrimitive (select (Variables|Default|GetIsSkeletalMesh) (Variables|Default|GetSkeletal_Mesh_Pickup) (Variables|Default|GetStatic_Mesh_Pickup))))
  (Variables|VFX|SetComponent%20Radius%20%28Scaled%29 (* (.SphereRadius (Collision|GetComponentBounds _output_get_3)) (Variables|VFX|GetComponent%20Radius%20Multiplier)))
  (Variables|VFX|SetComponent%20Radius (.SphereRadius (Collision|GetComponentBounds _output_get_3)))
  (CallFunction|SetupLight)
  ;[1 1 1] Is Blugo
  (if (Utilities|Equal(Object) (Variables|PickupProperties|GetStaticMesh) "/Game/Effects/Fort_Effects/BluGlow/BluGlowModel_NoLOD.BluGlowModel_NoLOD")
    (Variables|Default|SetIsBluglo true)
    (elif (Utilities|Equal(Object) (Variables|PickupProperties|GetStaticMesh) "/Game/Effects/Fort_Effects/BluGlow/Violentium/Violentium_Model_1.Violentium_Model_1")
      ;[1 1 1] is Violentium
      (Variables|Default|SetIs%20Violentium true))))
